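; Floor loot spawner actor that spawns pickups at its location from a selected loot tier group.
; SpawnFloorLoot runs once on the dedicated server, optionally from BeginPlay, waits for playlist data, then spawns the indexed tier's loot at a Z offset.
; Each pickup traces downward and, without a ground hit, has its movement's updated component cleared so it cannot fall through unloaded floors.
; The construction script shows the tier name, or "Invalid", on editor preview text.
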(event EventBeginPlay
  ;[0 0.15 0.1252 0.5] BeginPlay/SpawnFloorLoot: Spawn the item based on the desired type.
  (|CallParentFunction)
  (if (and (Variables|Settings|GetbForceSpawnOnBeginPlay) (Networking|IsDedicatedServer))
    (CallFunction|SpawnFloorLoot)))

(event Custom|SpawnFloorLoot
  (bind _loottiergroupnames (Variables|Default|GetLootTierGroupNames))
  (bind _lootindex (Variables|Loot|GetLootIndex))
  ;[0 0.15 0.1252 0.5] BeginPlay/SpawnFloorLoot: Spawn the item based on the desired type.
  (Utilities|FlowControl|DoOnce false
    (:Completed
      (if (Networking|IsDedicatedServer)
        (Class|FortAsyncActionCurrentPlaylistReady|PlaylistDataReadyAsync
          (:OnReady
            (if (Utilities|Array|IsValidIndex _loottiergroupnames _lootindex)
              (bind _returnvalue (Inventory|SpawnPickupInWorldWithLootTier :LootTierName (Utilities|Array|Get(aref) _loottiergroupnames _lootindex) :Position (Utilities|Operators|Add (Transformation|GetActorLocation) (Math|Vector|MakeVector 0.0 0.0 (Variables|Settings|GetPickupSpawnLocationOffset_Z))) :OverrideMaxStackCount 0 :bToss false :bTossWithVelocity false :SourceType "FloorLoot" :Source "Unset" :LootWorldLevel 0))
              ;;[0.0503 0.07522 0.15 0.5] If the ground underneath the pickup has not loaded yet, stop the pickup from updating its movement component so that it does not fall through the floor.
              (for _array_element _returnvalue
                (Utilities|IsValid _array_element
                  (:Is%20Valid
                    (bind (_outhit _returnvalue_2) (Collision|LineTraceByChannel (Transformation|GetActorLocation _array_element) (Utilities|Operators|Subtract (Transformation|GetActorLocation _array_element) (Math|Vector|MakeVector 0.0 0.0 (Variables|Settings|GetPickupGroundTraceDistance))) "TraceTypeQuery10" false (Utilities|Array|MakeArray _array_element)))
                    (if (not _returnvalue_2)
                      (bind _movementcomponent (Variables|/Script/FortniteGame.FortPickup|GetMovementComponent _array_element)
                        (:then
                          (Components|Movement|SetUpdatedComponent _movementcomponent 0))
                        (:else)))))))))))))

(fn ConstructionScript
  (bind _loottiergroupnames (Variables|Default|GetLootTierGroupNames))
  (bind _lootindex (Variables|Loot|GetLootIndex))
  (bind _editoronly_loottiergroupname (Variables|Default|GetEDITORONLY_LootTierGroupName))
  ;[0.15 0.06459 0.1491 0.5] ConstructionScript: Update preview.
  (if (Utilities|Array|IsValidIndex _loottiergroupnames _lootindex)
    (Rendering|Components|TextRender|SetText _editoronly_loottiergroupname (Utilities|Text|ToText(Name) (Utilities|Array|Get(acopy) _loottiergroupnames _lootindex)))
    (else
      (Rendering|Components|TextRender|SetText _editoronly_loottiergroupname (Utilities|Text|MakeLiteralText "INVTEXT(\"Invalid\")")))))
